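# Niagara master-material function that converts a 2D coordinate, defaulting to texture UVs, into polar coordinates.
# It offsets the input to center on 0.5, derives a scaled angle with Arctangent2Fast and a radius from the square root of summed squares.
# It outputs angle, radius and the scalar Param_1 as a three-component vector.
# A Custom HLSL atan2 node and a Rotator are present but unconnected to the output.
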MaterialFunction /Game/Effects/Niagara/MasterMaterials/Functions/UVs/PolarCoordinates
graph {nodes [
  n1 = Add(A=FunctionInput(Preview=TextureCoordinate()) {bUsePreviewValueAsDefault=true, inputName=Coordinate, inputType=FunctionInput_Vector2}) {constB=-0.5}
  MaterialExpressionComponentMask_20 = ComponentMask(None=@n1) {r=true}
  MaterialExpressionComponentMask_21 = ComponentMask(None=@n1) {g=true}
  Custom!(x=@MaterialExpressionComponentMask_20, y=@MaterialExpressionComponentMask_21) {
   code="return (atan2(y,x));"
   description=atan2
   outputType=CMOT_Float1
   inputs [
    {inputName=x, input {expression=&self:MaterialExpressionComponentMask_20}}
    {inputName=y, input {expression=&self:MaterialExpressionComponentMask_21}}
   ]
  }
  n2 = Divide(A=ComponentMask(None=Arctangent2Fast(Y=@MaterialExpressionComponentMask_21, X=@MaterialExpressionComponentMask_20)) {r=true})
  n3 = Multiply(A=@n1, B=@n1)
  n4 = SquareRoot(None=Add(A=ComponentMask(None=@n3) {r=true}, B=ComponentMask(None=@n3) {g=true}))
  FunctionOutput(None=AppendVector(A=AppendVector(A=@n2, B=@n4), B=ScalarParameter() {parameterName=Param_1}))
  Rotator()
 ]}
